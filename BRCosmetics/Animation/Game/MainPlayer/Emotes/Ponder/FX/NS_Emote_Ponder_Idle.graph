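# Niagara system for the Ponder emote idle: a warm glow of core, glare, ray and flare sprites, plus particle lights.
# Emitters share a system position offset, and most scale alpha by a system brightness phase, with looping particles holding the steady layers.
# Time of day scales glare, flare and one ray layer, which also fades with User.NormalizedNotifyProgress.
# Debug, mesh-ray and weapon-flash emitters are disabled.
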NiagaraSystem /BRCosmetics/Animation/Game/MainPlayer/Emotes/Ponder/FX/NS_Emote_Ponder_Idle
dependencies {
 usedDataInterfaces [&NiagaraDataInterfaceCurve, &NiagaraDataInterfaceVectorCurve]
 usedDynamicInputs [&Waveform, &RemapRange]
 usedModules [
  &self:SystemSpawnScript.SystemScriptSource.SystemScriptGraph.NiagaraNodeAssignment_2.SetVariables_#1
  &SystemState
  &self:SystemSpawnScript.SystemScriptSource.SystemScriptGraph.NiagaraNodeAssignment_0.SetVariables_#2
  &self:SystemSpawnScript.SystemScriptSource.SystemScriptGraph.NiagaraNodeAssignment_1.SetVariables_#3
 ]
}
emitters {
 Core001 {
  emitterUpdateScript [
   &EmitterState(Life Cycle Mode=::NewEnumerator1, Loop Behavior=::NewEnumerator1, UseLoopDelay=false)
   &SpawnBurst_Instantaneous(Use Spawn Probability=false, UseLoopCountLimit=false)
  ]
  particleSpawnScript [&InitializeParticle(
    Color=<rgba 1 0.5198 0.1382 1>
    Lifetime=0.125
    Position Offset=$System.SystemOffset
    Sprite Size=<80 30>
    Sprite Size Mode=::NewEnumerator4
    UsePositionOffset=true
    Write Material Random=true
   )]
  particleUpdateScript [
   &ParticleState(
    Kill Particles When Lifetime Has Elapsed=false
    Let Infinitely Lived Particles Die When Emitter Deactivates=true
    Loop Particles Lifetime=true
   )
   &SolveForcesAndVelocity(Clamp Velocity=false, Limit Acceleration=false)
   &CameraOffset(Camera Offset Amount=10)
   &ScaleColor(Scale Alpha=$System.LightBrightnessPhase, ScaleA=true, ScaleRGB=true)
  ]
  rendererClasses [&NiagaraSpriteRendererProperties]
  renderers [{rendererClass=&NiagaraSpriteRendererProperties}]
 }
 FN_Debug {
  bEnabled=false
  emitterUpdateScript [
   &EmitterState()
   &SpawnBurst_Instantaneous(Spawn Count=1, Use Spawn Probability=false)
  ]
  particleSpawnScript [/Niagara/Modules/Spawn/Initialization/InitializeParticle(
    Lifetime=5
    Module.Write Color=true
    Module.Write Lifetime=true
    Module.Write Mass=true
    Module.Write Material Random=true
    Module.Write Position=false
    Module.Write Scale=false
    Module.Write SpriteRotation=false
    Module.Write SpriteSize=true
    Position=&SimulationPosition()
    Sprite Size=<100 100>
   )]
  particleUpdateScript [
   &ParticleState()
   /SetVariables(Particles.Debug:NiagaraFloat=&MakeFloatFromVector(Vector=$Engine.Owner.Scale, Channel=::NewEnumerator0)) {isSetParameters=true}
   &DynamicMaterialParameters(
    Index 0 Param 1=$Particles.Debug
    Module.Param0WriteEnabled=true
    Module.Param1WriteEnabled=true
    Module.Param2WriteEnabled=true
    Module.Param3WriteEnabled=true
   )
  ]
  rendererClasses [&NiagaraSpriteRendererProperties]
  renderers [{rendererClass=&NiagaraSpriteRendererProperties}]
 }
 FN_Sprite {
  emitterUpdateScript [
   &EmitterState(Life Cycle Mode=::NewEnumerator1, Loop Behavior=::NewEnumerator1, UseLoopDelay=false)
   &SpawnBurst_Instantaneous(Use Spawn Probability=false, UseLoopCountLimit=false)
  ]
  particleSpawnScript [/Niagara/Modules/Spawn/Initialization/V2/InitializeParticle(
    Color=<rgba 1 0.6804 0.202 0.2>
    Lifetime=0.2
    Position Offset=$System.SystemOffset
    Sprite Size Mode=::NewEnumerator3
    Uniform Sprite Size=120
    UsePositionOffset=true
    Write Material Random=true
   ) {module=InitializeParticle001}]
  particleUpdateScript [
   &ParticleState(
    Kill Particles When Lifetime Has Elapsed=false
    Let Infinitely Lived Particles Die When Emitter Deactivates=true
    Loop Particles Lifetime=true
   )
   &ScaleSpriteSize(Scale Factor=&Vector2DFromFloat(Value=&FloatFromCurve(
      FloatCurve=DataInterface() {propertyValues {Curve {keys [{arriveTangent=4.5137, interpMode=RCIM_Cubic, leaveTangent=4.5137, tangentMode=RCTM_User, time=0, value=0}, <at 1 0.5 RCIM_Cubic>]}}}
      CurveIndex=&Multiply_Float(A=$Engine.System.Age, B=2)
     )))
   &SolveForcesAndVelocity(Clamp Velocity=false, Limit Acceleration=false)
   &ScaleColor(Scale Alpha=$System.LightBrightnessPhase, ScaleA=true, ScaleRGB=true)
   &FN_Light_Attributes(
    Color=$Particles.Color
    Create Color Attribute=true
    Scale Color=&Add_Float(
     A=20
     B=&FloatFromCurve(
      FloatCurve=DataInterface() {propertyValues {Curve {keys [
          {arriveTangent=9.0725, interpMode=RCIM_Cubic, leaveTangent=9.0725, tangentMode=RCTM_User, time=0, value=0}
          {arriveTangent=-2.5249, interpMode=RCIM_Cubic, leaveTangent=-2.5249, tangentMode=RCTM_User, time=0.05, value=1}
          {arriveTangent=-0.0405, interpMode=RCIM_Cubic, leaveTangent=-0.0405, tangentMode=RCTM_User, time=0.5, value=0}
         ]}}}
      CurveIndex=$System.Age
      Scale Curve=25
     )
    )
   )
   &FN_Light_Attributes(
    Color=$Particles.Color
    Create Color Attribute=true
    Create Position Attribute=true
    Enable Particle Light Enable=true
    Light Enabled=true
    Position=<0 0 -10>
    Scale Color=0.075
   ) {module=FN_Light_Attributes001}
  ]
  rendererClasses [
   &NiagaraSpriteRendererProperties
   &NiagaraLightRendererProperties
   &NiagaraComponentRendererProperties
  ]
  renderers [
   {rendererClass=&NiagaraSpriteRendererProperties}
   {rendererClass=&NiagaraLightRendererProperties, rendererIndex=1}
   {rendererClass=&NiagaraComponentRendererProperties, rendererIndex=2}
  ]
 }
 FN_Weapons_Core_Flash_BETA {
  bEnabled=false
  emitterUpdateScript [
   &EmitterState()
   &SpawnBurst_Instantaneous(Spawn Count=1, Use Spawn Probability=false)
  ]
  particleSpawnScript [
   /Niagara/Modules/Spawn/Initialization/InitializeParticle(
    Lifetime=&UniformRangedFloat(Minimum=0.1, Maximum=0.1)
    Mesh Scale=&UniformRangedVector(Minimum=<0.6 0.6 0.6>, Maximum=<0.9 0.8 0.8>)
    Module.Write Color=true
    Module.Write Lifetime=true
    Module.Write Mass=true
    Module.Write Material Random=true
    Module.Write Position=true
    Module.Write Scale=true
    Module.Write SpriteRotation=false
    Module.Write SpriteSize=true
    Position=&Add_Vector(A=/Niagara/DynamicInputs/Helpers/SimulationPosition(), B=<0 0 0>)
    Sprite Size=<100 100>
   )
   /Game/Effects/Niagara/Modules/Orientation/FN_Fan(Spin:NiagaraFloat=&UniformRangedFloat(Minimum=0, Maximum=1))
   /SetVariables(Particles.LightPositionNew:Vector3f=&Add_Vector(
     A=Linked() {linkedVariable {name=Particles.Position, type=NiagaraPosition}}
     B=/Niagara/DynamicInputs/Transforms/TransformPosition(Position=<25 0 0>, Source Space=::Local, Destination Space=::Simulation)
    )) {isSetParameters=true}
  ]
  particleUpdateScript [
   &ParticleState()
   &Color(
    Color=<rgba 1 0.3636 0.01 1>
    Scale Color=&VectorFromCurve(VectorCurve=DataInterface() {propertyValues {XCurve {keys [<at 0.7 12>,<at 1 1>]}, YCurve {keys [<at 0.7 12>,<at 1 1>]}, ZCurve {keys [<at 0.7 12>,<at 1 1>]}}})
   )
   &ScaleMeshSize(Scale Factor=&VectorFromFloat(Value=&FloatFromCurve(FloatCurve=DataInterface() {propertyValues {Curve {keys [{arriveTangent=2.2457, interpMode=RCIM_Cubic, leaveTangent=2.2457, tangentMode=RCTM_User, time=0, value=0}, <at 1 1>]}}})))
   &FN_Sprite_Attributes(
    Create SpriteSize Attribute=true
    SpriteSize=&Vector2DFromFloat(Value=50)
    Write to Module Namespace=true
   ) {module=FN_Sprite}
   /SetVariables(Particles.SpriteColor=&MultiplyLinearColor(Linear Color=$Particles.Color, Scale Factor=<1 1 1 0.2>)) {isSetParameters=true}
   &DynamicMaterialParameters(Write Parameter Index 0=false)
   &FN_Light_Attributes(
    Color=<rgba 1 0.6601 0.0729 1>
    Create Color Attribute=true
    Create Exponent Attribute=true
    Enable Particle Light Enable=true
    Exponent=4
    Light Enabled=true
    Scale Color=5
   )
   &FN_Light_Attributes(Color=<rgba 1 0.7728 0.204 1>, Create Color Attribute=true, Enable Particle Light Enable=true, Light Enabled=true) {module=FN_Light_Attributes001}
  ]
  rendererClasses [
   &NiagaraComponentRendererProperties
   &NiagaraMeshRendererProperties
   &NiagaraLightRendererProperties
   &NiagaraSpriteRendererProperties
  ]
  renderers [
   {rendererClass=&NiagaraComponentRendererProperties, rendererIndex=0}
   {rendererClass=&NiagaraMeshRendererProperties, rendererIndex=1}
   {rendererClass=&NiagaraLightRendererProperties, rendererIndex=2}
   {rendererClass=&NiagaraSpriteRendererProperties, rendererIndex=3}
  ]
 }
 Glare {
  emitterUpdateScript [
   &EmitterState(Life Cycle Mode=::NewEnumerator1, Loop Behavior=::NewEnumerator1, UseLoopDelay=false)
   &SpawnBurst_Instantaneous(Use Spawn Probability=false, UseLoopCountLimit=false)
  ]
  particleSpawnScript [/Niagara/Modules/Spawn/Initialization/V2/InitializeParticle(
    Color=<rgba 0.7 0.5046 0.1749 1>
    Lifetime=0.1
    Position Offset=$System.SystemOffset
    Sprite Size Mode=::NewEnumerator3
    Uniform Sprite Size=60
    UsePositionOffset=true
    Write Material Random=true
   )]
  particleUpdateScript [
   &ParticleState(
    Kill Particles When Lifetime Has Elapsed=false
    Let Infinitely Lived Particles Die When Emitter Deactivates=true
    Loop Particles Lifetime=true
   )
   &SolveForcesAndVelocity(Clamp Velocity=false, Limit Acceleration=false)
   &CameraOffset(Camera Offset Amount=10)
   &ScaleColor(
    Scale Alpha=$System.LightBrightnessPhase
    Scale RGB=&Multiply_VectorByFloat(Vector=<1 1 1>, Float=&FNTimeOfDayScalar(Night Scale=0.5, Morning Scale=0.5, Evening Scale=0.7))
    ScaleA=true
    ScaleRGB=true
   )
  ]
  rendererClasses [&NiagaraSpriteRendererProperties]
  renderers [{rendererClass=&NiagaraSpriteRendererProperties}]
 }
 GlowFlare001 {
  emitterUpdateScript [
   &EmitterState(Life Cycle Mode=::NewEnumerator1, Loop Behavior=::NewEnumerator1, Loop Duration=1, UseLoopDelay=true)
   &SpawnBurst_Instantaneous(Spawn Count=1, Use Spawn Probability=false)
  ]
  particleSpawnScript [&InitializeParticle(
    Color=<rgba 1 0.7407 0.3582 1>
    Lifetime=0.25
    Position Offset=$System.SystemOffset
    Sprite Size Mode=::NewEnumerator3
    Uniform Sprite Size=22
    UsePositionOffset=true
    Write Material Random=true
   )]
  particleUpdateScript [
   &ParticleState()
   &ScaleColor(
    Scale Alpha=&Multiply_Float(
     A=&FloatFromCurve(FloatCurve=DataInterface() {propertyValues {Curve {keys [{interpMode=RCIM_Cubic, leaveTangent=-0.6434, tangentMode=RCTM_Break, time=0, value=1}, <at 1 0>]}}})
     B=&FNTimeOfDayScalar(Night Scale=1, Morning Scale=0.3, Evening Scale=0.5)
    )
    ScaleA=true
    ScaleRGB=true
   )
   &CameraOffset(Camera Offset Amount=50)
   &ScaleSpriteSize(
    Uniform Curve Scale=3
    Uniform Curve Sprite Scale=DataInterface() {propertyValues {Curve {keys [{tangentMode=RCTM_User, time=0, value=0}, <at 1 1>]}}}
   )
  ]
  rendererClasses [&NiagaraSpriteRendererProperties]
  renderers [{rendererClass=&NiagaraSpriteRendererProperties}]
 }
 Rays {
  bEnabled=false
  emitterUpdateScript [
   &EmitterState()
   &SpawnRate(
    SpawnRate=&FloatFromCurve(
     FloatCurve=DataInterface() {propertyValues {Curve {keys [<at 0 1 RCIM_Cubic>, {arriveTangent=-2, interpMode=RCIM_Cubic, leaveTangent=-2, time=0.2, value=1}, <at 0.5 0 RCIM_Cubic>]}}}
     CurveIndex=$User.NormalizedNotifyProgress
     Scale Curve=16
    )
    Use Spawn Probability=false
   )
   &SpawnBurst_Instantaneous(Spawn Count=5, Use Spawn Probability=false) {enabled=false}
  ]
  particleSpawnScript [
   /Niagara/Modules/Spawn/Initialization/InitializeParticle(
    Color=<rgba 1 0.579 0.1407 0.1>
    Lifetime=&UniformRangedFloat(Minimum=1.5, Maximum=2)
    Mesh Scale=&VectorFromFloat(Value=0.45)
    Module.Write Color=true
    Module.Write Lifetime=true
    Module.Write Mass=true
    Module.Write Material Random=true
    Module.Write Position=false
    Module.Write Scale=true
    Module.Write SpriteRotation=false
    Module.Write SpriteSize=false
    Position=/Niagara/Functions/Localspace/SimulationPosition()
   )
   &AddVelocity(Module.Use Falloff Distance=false, Override Seed=false, Velocity Mode=::NewEnumerator1, Velocity Speed=0.1)
   &AddVelocityFromPoint(
    Module.Use Falloff Distance=false
    Origin Offset=<0 0 -1.0218>
    Velocity Origin:Vector3f=&SimulationPosition()
    Velocity Strength=1
   ) {enabled=false}
  ]
  particleUpdateScript [
   &ParticleState()
   &SolveForcesAndVelocity(Clamp Velocity=false, Limit Acceleration=false)
   &ScaleColor(
    Scale Alpha=&FloatFromCurve(FloatCurve=DataInterface() {propertyValues {Curve {keys [
         {interpMode=RCIM_Cubic, tangentMode=RCTM_User, time=0, value=0}
         {interpMode=RCIM_Cubic, tangentMode=RCTM_User, time=0.5, value=1}
         {tangentMode=RCTM_User, time=1, value=0}
        ]}}})
    ScaleA=true
    ScaleRGB=true
   )
   &ScaleMeshSize(Scale Factor=&VectorFromFloat(Value=&FloatFromCurve(
      FloatCurve=DataInterface() {propertyValues {Curve {keys [<at 0 0 RCIM_Cubic>, {arriveTangent=0.002, interpMode=RCIM_Cubic, leaveTangent=0.002, time=0.5, value=1}, <at 1.0011 0.002 RCIM_Cubic>]}}}
      Scale Curve=1.5
     ))) {enabled=false}
  ]
  rendererClasses [&NiagaraMeshRendererProperties]
  renderers [{rendererClass=&NiagaraMeshRendererProperties, rendererIndex=0}]
 }
 Rays001 {
  emitterUpdateScript [
   &EmitterState(Life Cycle Mode=::NewEnumerator1, Loop Behavior=::NewEnumerator1, UseLoopDelay=false)
   &SpawnBurst_Instantaneous(Use Spawn Probability=false, UseLoopCountLimit=false)
  ]
  particleSpawnScript [/Niagara/Modules/Spawn/Initialization/V2/InitializeParticle(
    Color=<rgba 0.7 0.5046 0.1749 1>
    Lifetime=0.1
    Position Offset=$System.SystemOffset
    Sprite Size Mode=::NewEnumerator3
    Uniform Sprite Size=60
    UsePositionOffset=true
    Write Material Random=true
   )]
  particleUpdateScript [
   &ParticleState(
    Kill Particles When Lifetime Has Elapsed=false
    Let Infinitely Lived Particles Die When Emitter Deactivates=true
    Loop Particles Lifetime=true
   )
   &SolveForcesAndVelocity(Clamp Velocity=false, Limit Acceleration=false)
   &CameraOffset(Camera Offset Amount=-10)
   &ScaleColor(
    Scale Alpha=&Multiply_Float(
     A=$System.LightBrightnessPhase
     B=&FloatFromCurve(
      FloatCurve=DataInterface() {propertyValues {Curve {keys [<at 0 1 RCIM_Cubic>, {arriveTangent=-1.1111, interpMode=RCIM_Cubic, leaveTangent=-1.1111, time=0.6, value=1}, <at 0.9 0>, <at 1 0>]}}}
      CurveIndex=$User.NormalizedNotifyProgress
     )
    )
    Scale RGB=&Multiply_VectorByFloat(Vector=<1 1 1>, Float=&FNTimeOfDayScalar(Night Scale=0.5, Morning Scale=0.5, Evening Scale=0.7))
    ScaleA=true
    ScaleRGB=true
   )
   &DynamicMaterialParameters(
    Index 0 Param 1=&Multiply_Float(A=$Emitter.Age, B=1)
    Index 0 Param 2=-0.05
    Module.Param0WriteEnabled=true
    Module.Param1WriteEnabled=true
    Module.Param2WriteEnabled=true
    Module.Param3WriteEnabled=true
   )
  ]
  rendererClasses [&NiagaraSpriteRendererProperties]
  renderers [{rendererClass=&NiagaraSpriteRendererProperties}]
 }
 Rays002 {
  emitterUpdateScript [
   &EmitterState(Life Cycle Mode=::NewEnumerator1, Loop Behavior=::NewEnumerator1, UseLoopDelay=false)
   &SpawnBurst_Instantaneous(Use Spawn Probability=false, UseLoopCountLimit=false)
  ]
  particleSpawnScript [&InitializeParticle(
    Color=<rgba 0.7 0.5046 0.1749 1>
    Lifetime=0.35
    Position Offset=$System.SystemOffset
    Sprite Rotation Mode=::NewEnumerator0
    Sprite Size Mode=::NewEnumerator3
    Uniform Sprite Size=60
    UsePositionOffset=true
    Write Material Random=true
   )]
  particleUpdateScript [
   &ParticleState()
   &ScaleSpriteSize(
    Uniform Curve Scale=3
    Uniform Curve Sprite Scale=DataInterface() {propertyValues {Curve {keys [
        {arriveTangent=4.1785, interpMode=RCIM_Cubic, leaveTangent=4.1785, tangentMode=RCTM_User, time=0, value=0}
        {arriveTangent=1.1179, interpMode=RCIM_Cubic, leaveTangent=1.1179, tangentMode=RCTM_User, time=0.2042, value=0.6228}
        {arriveTangent=0.0224, interpMode=RCIM_Cubic, leaveTangent=0.0224, tangentMode=RCTM_User, time=1, value=1}
       ]}}}
   )
   &SolveForcesAndVelocity(Clamp Velocity=false, Limit Acceleration=false)
   &CameraOffset(Camera Offset Amount=-10)
   &ScaleColor(
    Scale Alpha=&Multiply_Float(
     A=$System.LightBrightnessPhase
     B=&FloatFromCurve(FloatCurve=DataInterface() {propertyValues {Curve {keys [{arriveTangent=-1.404, interpMode=RCIM_Cubic, leaveTangent=-1.404, tangentMode=RCTM_User, time=0, value=1}, <at 1 0>]}}})
    )
    ScaleA=true
    ScaleRGB=true
   )
   &DynamicMaterialParameters(
    Index 0 Param 1=&Multiply_Float(A=$Emitter.Age, B=1)
    Index 0 Param 2=-0.05
    Module.Param0WriteEnabled=true
    Module.Param1WriteEnabled=true
    Module.Param2WriteEnabled=true
    Module.Param3WriteEnabled=true
   )
  ]
  rendererClasses [&NiagaraSpriteRendererProperties]
  renderers [{rendererClass=&NiagaraSpriteRendererProperties}]
 }
}
summary {
 systemName=NS_Emote_Ponder_Idle
 emitters [
  {bEnabled=false, emitterName=FN_Debug, rendererClasses [&NiagaraSpriteRendererProperties]}
  {
   emitterName=FN_Sprite
   rendererClasses [
    &NiagaraSpriteRendererProperties
    &NiagaraLightRendererProperties
    &NiagaraComponentRendererProperties
   ]
  }
  {emitterName=Glare, rendererClasses [&NiagaraSpriteRendererProperties]}
  {bEnabled=false, emitterName=Rays, rendererClasses [&NiagaraMeshRendererProperties]}
  {emitterName=Core001, rendererClasses [&NiagaraSpriteRendererProperties]}
  {emitterName=Rays001, rendererClasses [&NiagaraSpriteRendererProperties]}
  {emitterName=Rays002, rendererClasses [&NiagaraSpriteRendererProperties]}
  {emitterName=GlowFlare001, rendererClasses [&NiagaraSpriteRendererProperties]}
  {
   bEnabled=false
   emitterName=FN_Weapons_Core_Flash_BETA
   rendererClasses [
    &NiagaraComponentRendererProperties
    &NiagaraMeshRendererProperties
    &NiagaraLightRendererProperties
    &NiagaraSpriteRendererProperties
   ]
  }
 ]
 userVariables [
  User.ConeAngleFX {type=NiagaraFloat, defaultValue=2, description=""}
  User.ConeRotationFX {type=NiagaraFloat, defaultValue=0.1, description=""}
  User.HueFX {type=NiagaraFloat, defaultValue=0, description=""}
  User.NormalizedNotifyProgress {type=NiagaraFloat, defaultValue=0, description=""}
 ]
}
